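# Default-lit material for the Papaya cuddle statue, with glowing eyes and a height-weighted emissive.
# Base color is the contrast-adjusted diffuse times a bottom-to-top tint from bounding-box height.
# The packed mask drives metallic and specular (G), roughness (B), eye glow (A) and extra emissive (inverted R).
# Emissive is scaled by view-facing normal and height, then passed through MF_TOD_EmissiveColors, presumably for time of day.
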Material /PapayaGameplay/Environments/Papaya/Materials/M_Papaya_Statue
graph {
 properties {bUseMaterialAttributes=true, pixelDepthOffsetMode=PDOM_Legacy, usedShadingModels=Default Lit}
 nodes [
  n1 = Power(
   Base=TextureSampleParameter2D() {parameterName=Diffuse, texture=/PapayaGameplay/Environments/Papaya/Textures/T_Papaya_CuddleStatue_D}.RGB
   Exp=ScalarParameter() {defaultValue=0.4, parameterName=BaseColorContrast}
  )
  n2 = MaterialFunctionCall() {materialFunction=&BoundingBoxBased_0-1_UVW}
  n3 = LinearInterpolate(
   A=VectorParameter() {parameterName=StatueColorBottom, defaultValue=<rgba 0.3906 0.3352 0.3196 0>}.RGB
   B=VectorParameter() {parameterName=StatueColorTop, defaultValue=<rgba 0.6562 0.5674 0.5674 0>}.RGB
   Alpha=Power(Base=@n2.B, Exp=ScalarParameter() {defaultValue=3, parameterName=zContrastColor})
  )
  n4 = TextureSampleParameter2D() {parameterName=Specular, samplerType=SAMPLERTYPE_Masks, texture=/PapayaGameplay/Environments/Papaya/Textures/T_Papaya_CuddleStatue_S}
  n5 = ScalarParameter() {parameterName=RoughnessMin}
  n6 = ScalarParameter() {defaultValue=0.8, parameterName=RoughnessMax}
  n7 = ScalarParameter() {defaultValue=1, parameterName=EyeEmissiveBoost}
  n8 = TextureSampleParameter2D() {parameterName=Normal, samplerType=SAMPLERTYPE_Normal, texture=/PapayaGameplay/Environments/Papaya/Textures/T_Papaya_CuddleStatue_N}
  n9 = GetMaterialAttributes(None=MakeMaterialAttributes!(
    BaseColor=Multiply(A=@n1, B=@n3)
    Metallic=@n4.G
    Specular=@n4.G
    Roughness=LinearInterpolate(A=@n5, B=@n6, Alpha=@n4.B)
    Anisotropy
    EmissiveColor=Multiply(A=@n4.A, B=@n7)
    Opacity
    OpacityMask
    Normal=@n8.RGB
    Tangent
    WorldPositionOffset
    SubsurfaceColor
    ClearCoat
    ClearCoatRoughness
    AmbientOcclusion
    Refraction
    CustomizedUVs_0
    CustomizedUVs_1
    CustomizedUVs_2
    CustomizedUVs_3
    CustomizedUVs_4
    CustomizedUVs_5
    CustomizedUVs_6
    CustomizedUVs_7
    PixelDepthOffset
    ShadingModel
    Displacement
   )) {attributeGetTypes [#1]}
  n10 = OneMinus(None=@n4.R)
  n11 = ScalarParameter() {defaultValue=1, parameterName=SpecAsEmissiveBoost}
  n12 = Multiply(A=Add(A=@n9.Emissive Color, B=Multiply(A=@n10, B=@n11)), B=DotProduct(A=CameraVectorWS(), B=PixelNormalWS()))
  n13 = Power(Base=@n2.B, Exp=ScalarParameter() {defaultValue=2.5627, parameterName=zContrastEmissive})
  MaterialFunctionCall!(In=SetMaterialAttributes(MaterialAttributes=@n9.MaterialAttributes, Emissive Color=Multiply(A=@n12, B=@n13)) {attributeSetTypes [#1]}) {materialFunction=/Game/Athena/WinterArt/Materials/MF_TOD_EmissiveColors}
 ]
}
