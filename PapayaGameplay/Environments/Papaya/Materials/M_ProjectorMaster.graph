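# M_ProjectorMaster is a Nanite-enabled, Default Lit master material for a holographic projector prop in the Papaya environment.
# It uses the projector's diffuse and normal textures for the surface.
# The blue channel of the MRS texture serves as a glow mask, multiplied by EmissiveColor (orange by default) and EmissiveBrightness, so instances can tune the glow.
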Material /PapayaGameplay/Environments/Papaya/Materials/M_ProjectorMaster
graph {
 properties {bUsedWithNanite=true, pixelDepthOffsetMode=PDOM_Legacy, usedShadingModels=Default Lit}
 nodes [
  Multiply(
   A=Multiply(
    A=TextureSample() {texture=/PapayaGameplay/Environments/Papaya/Textures/T_HoloProjector_MRS}.B
    B=VectorParameter() {parameterName=EmissiveColor, defaultValue=<rgba 1 0.3857 0 1>}.RGB
   )
   B=ScalarParameter() {defaultValue=10, parameterName=EmissiveBrightness}
  )
  TextureSample() {samplerType=SAMPLERTYPE_Normal, texture=/PapayaGameplay/Environments/Papaya/Textures/T_HoloProjector_Normal}
  TextureSample() {texture=/PapayaGameplay/Environments/Papaya/Textures/T_HoloProjector_Diffuse}
 ]
}
