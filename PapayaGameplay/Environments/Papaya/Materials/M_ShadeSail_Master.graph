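# Master subsurface material for Papaya shade sails that ripple in the wind and react to event audio.
# Gadget color, offset by distance from a stage location, drives base and view-dependent subsurface color.
# Scrolling noise animates fold normals and masked world position offset, with an audio-spectrum pump along vertex normals.
# Emissive combines spectrum-driven smoke in an expanding sphere mask with an amplitude-driven EQ grid hue shift that fades with stage distance.
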Material /PapayaGameplay/Environments/Papaya/Materials/M_ShadeSail_Master
graph {
 properties {
  bUseMaterialAttributes=true
  bUsedWithNanite=true
  pixelDepthOffsetMode=PDOM_Legacy
  shadingModel=MSM_Subsurface
  usedShadingModels=Subsurface
 }
 nodes [
  n1 = ComponentMask(None=WorldPosition().XYZ) {g=true, r=true}
  n2 = Add(
   A=Divide(A=@n1, B=ScalarParameter() {defaultValue=1024, parameterName=WPONoiseSize})
   B=Multiply(A=Time(), B=ScalarParameter() {defaultValue=0.2, parameterName=WindSpeed})
  )
  n3 = TextureSample(UVs=@n2) {samplerType=SAMPLERTYPE_Grayscale, texture=&LowResBlurredNoise}
  Add(A=@n3.R, B=ScalarParameter() {defaultValue=-0.5, parameterName=WPONoiseBias})
  n4 = CollectionParameter() {
   collection=&FortniteMaterialParameters
   parameterName=PapayaBaseColorParameter
  }
  n5 = CollectionParameter() {
   collection=&FortniteMaterialParameters
   parameterName=PapayaBaseColorParameterOffset
  }
  n6 = MaterialFunctionCall(Vector 3=Subtract(
    A=ObjectPositionWS()
    B=VectorParameter() {group=Distance Falloff, parameterName=Stage Location, defaultValue=<rgba -1024 26350 -3450 1>}.RGB
   )) {materialFunction=/Engine/Functions/Engine_MaterialFunctions02/Utility/VectorLength}
  n7 = ScalarParameter() {defaultValue=1000, group=Distance Falloff, parameterName=Radius Min, sortPriority=10}
  n8 = ScalarParameter() {defaultValue=10000, group=Distance Falloff, parameterName=Radius Max, sortPriority=20}
  n9 = Saturate(None=Divide(A=Subtract(A=@n6.V3 Length, B=@n7), B=Subtract(A=@n8, B=@n7)))
  n10 = MaterialFunctionCall!(
   Media Player TexCoord
   Media Color Code Coordinate
   Media Texture Resolution
   Base Color Parameter=Add(A=@n4, B=Multiply(A=@n9, B=@n5) {constB=0.2})
   Base Color Mix
   Media Color Code Mix
   Intermission Color Parameter
   Intermission Color Mix
   Desaturation
  ) {materialFunction=/Game/Environments/Papaya/Materials/MaterialFunctions/MF_Papaya_GadgetColor}
  n11 = TextureSampleParameter2D() {parameterName=Diffuse, texture=/PapayaGameplay/Environments/Papaya/Textures/T_Papaya_ShadeSail_D}
  n12 = Multiply(A=@n10.Result, B=LinearInterpolate(A=ScalarParameter() {defaultValue=0.8, parameterName=BaseColorEdgeDarkness}, Alpha=@n11.R))
  n13 = Multiply(A=@n12, B=ScalarParameter() {defaultValue=0.8, parameterName=BaseColorBrightness})
  n14 = TextureSampleParameter2D() {parameterName=SpecularMasks, samplerType=SAMPLERTYPE_Masks, texture=/PapayaGameplay/Environments/Papaya/Textures/T_Papaya_ShadeSail_S}
  n15 = TextureSample(UVs=TextureCoordinate() {uTiling=8, vTiling=8}) {samplerType=SAMPLERTYPE_LinearColor, texture=/Game/UI/Frontend/ItemShop/Com/vfx/T_Pattern_Triangles}
  n16 = TextureSampleParameter2D() {parameterName=Normals, samplerType=SAMPLERTYPE_Normal, texture=/PapayaGameplay/Environments/Papaya/Textures/T_Papaya_ShadeSail_N}
  n17 = Power(Base=@n12, Exp=ScalarParameter() {defaultValue=1.2, parameterName=SubsrfaceColorExponent})
  n18 = Power(Base=DotProduct(A=PixelNormalWS(), B=CameraVectorWS()), Exp=ScalarParameter() {defaultValue=1, parameterName=ViewDependantExponent})
  n19 = SetMaterialAttributes!(
   MaterialAttributes
   Base Color=@n13
   Metallic=@n14.G
   Specular=@n14.G
   Roughness=Add(A=@n14.B, B=LinearInterpolate(Alpha=@n15.G) {constA=-0.1, constB=0.1})
   Normal=@n16.RGB
   Opacity=OneMinus(None=@n11.A)
   Subsurface Color=Multiply(A=Multiply(A=@n17, B=@n18), B=@n14.A)
  ) {attributeSetTypes [
    #1
    #2
    #3
    #4
    #5
    #6
    #7
   ]}
  n20 = GetMaterialAttributes(None=@n19) {attributeGetTypes [#5]}
  n21 = Multiply(
   A=Multiply(
    A=TextureSample(UVs=@n2) {samplerType=SAMPLERTYPE_Normal, texture=/Game/Gadgets/Assets/VinderTech_GliderChute/Glider_BlackMondayCape/Textures/T_Folds_N}.RGB
    B=Constant3Vector() {constant=<rgba 0.5 0.5 0 0>}
   )
   B=@n14.A
  )
  n22 = Transform(None=Constant3Vector() {constant=<rgba 0 0 1 0>}) {transformSourceType=TRANSFORMSOURCE_Local}
  n23 = CollectionParameter() {collection=&FortniteMaterialParameters, parameterName=EventAudio_Spectrum}
  n24 = Power(Base=Clamp(None=ComponentMask(None=@n23) {r=true}), Exp=ScalarParameter() {defaultValue=3, parameterName=percExponent})
  n25 = VertexNormalWS()
  n26 = ScalarParameter() {parameterName=WPOPumpDisplacement}
  n27 = Add(
   A=Multiply(A=Multiply(A=@n3.R, B=@n22), B=ScalarParameter() {defaultValue=50, parameterName=WPODisplacement})
   B=Multiply(A=@n24, B=Multiply(A=@n25, B=@n26))
  )
  n28 = Constant2Vector() {g=10000, r=5000}
  n29 = Time()
  n30 = Constant2Vector() {g=1}
  n31 = TextureSample(UVs=Add(A=Divide(A=@n1, B=@n28), B=Multiply(A=@n29, B=@n30))) {texture=&T_Inky_Smoke_Tile_Chromatic}
  n32 = SphereMask(
   A=ComponentMask(None=MaterialFunctionCall() {materialFunction=&BoundingBoxBased_0-1_UVW}.RGB) {g=true, r=true}
   B=Constant() {r=0.5}
   Radius=LinearInterpolate(A=ScalarParameter() {defaultValue=0.5, parameterName=rMin}, B=ScalarParameter() {defaultValue=4, parameterName=r}, Alpha=@n24)
   Hardness=Constant()
  )
  n33 = Multiply(A=Multiply(A=LinearInterpolate(B=@n31.RGB) {constA=0.5}, B=Multiply(A=@n24, B=@n32)), B=Multiply(A=@n14.A, B=Multiply(A=@n18, B=@n11.A)))
  n34 = TextureCoordinate()
  n35 = CollectionParameter() {collection=&FortniteMaterialParameters, parameterName=EventAudio_Amplitude}
  n36 = ScalarParameter() {defaultValue=20, group=Grid, parameterName=Tiles Y, sortPriority=-90}
  n37 = Multiply(A=Multiply(A=@n35, B=ScalarParameter() {defaultValue=2, group=Grid, parameterName=Grid Speed, sortPriority=-30}), B=@n36)
  n38 = Panner(Coordinate=@n34, Time=Divide(A=Floor(None=@n37), B=@n36), Speed=Constant2Vector() {g=1})
  n39 = MaterialFunctionCall(X=ScalarParameter() {defaultValue=20, group=Grid, parameterName=Tiles X, sortPriority=-100}, Y=@n36) {materialFunction=&MakeFloat2}
  n40 = TextureSample(UVs=Multiply(
    A=MaterialFunctionCall!(TexCoord=@n38, Tiles XY=@n39.Result) {materialFunction=/Game/Environments/Papaya/Materials/MF_GridTexCoord}.TexCoord
    B=Constant2Vector() {g=3, r=3}
   )) {samplerType=SAMPLERTYPE_LinearGrayscale, texture=/PapayaGameplay/Environments/Papaya/Textures/T_Papaya_Shade_EQ_Noise}
  n41 = OneMinus(None=Multiply(A=OneMinus(None=@n40.R), B=ScalarParameter() {defaultValue=0.5, parameterName=EQ Noise Gain}))
  n42 = TextureSampleParameter2D(UVs=MaterialFunctionCall(TexCoord=@n34, Tiles XY=@n39.Result) {materialFunction=&MF_GridTexCoord}.TexCoord) {
   group=Grid
   parameterName=Ramp Texture
   samplerType=SAMPLERTYPE_LinearGrayscale
   sortPriority=-80
   texture=/PapayaGameplay/Environments/Papaya/Textures/T_Papaya_Shade_EQ_Ramp
  }
  n43 = Power(Base=@n42.R, Exp=ScalarParameter() {defaultValue=2, group=Grid, parameterName=Ramp Exponent, sortPriority=-60}) {constExponent=0.5}
  n44 = OneMinus(None=@n35)
  n45 = Saturate(None=MaterialFunctionCall(
    Input=Multiply(A=@n41, B=@n43)
    Input Low=Add(A=@n44) {constB=-0.2}
    Input High=Add(A=@n44) {constB=0.2}
    Target Low=Constant()
    Target High=Constant() {r=1}
   ) {materialFunction=&RemapValueRange}.Result)
  n46 = ScalarParameter() {defaultValue=0.7, parameterName=Hue Shift Gain}
  n47 = LinearInterpolate(
   A=MaterialFunctionCall(Hue Shift Percentage=Multiply(A=@n45, B=@n46), Texture=@n10.Result) {materialFunction=&HueShift}.Result
   B=@n13
   Alpha=Power(Base=@n9, Exp=ScalarParameter() {defaultValue=0.6, group=Distance Falloff, parameterName=EQ Falloff Exponent, sortPriority=40})
  )
  SetMaterialAttributes!(
   MaterialAttributes=SetMaterialAttributes!(MaterialAttributes=@n20.MaterialAttributes, Normal=Add(A=@n20.Normal, B=@n21), World Position Offset=Multiply(A=@n14.A, B=@n27)) {attributeSetTypes [#5, #8]}
   Emissive Color=Multiply(A=Multiply(A=@n33, B=@n47), B=ScalarParameter() {defaultValue=10, parameterName=EmissiveBrightness})
  ) {attributeSetTypes [#9]}
 ]
}
